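# Translucent unlit material for the tail VFX of the Stamina Male glider, built for Niagara mesh particles and ribbons.
# Color layers three axis-flattened NdotL lighting bands, darkens a custom-depth edge-detected outline scaled by distance and FOV, then applies UV- and particle-driven desaturation and boost.
# Dynamic parameters drive MF_NiagaraFollowingDeform vertex bending, with a DisableWPO switch for normal-offset motion.
# Opacity combines UV gradient, depth fade, view-alignment fade and particle alpha.
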Material /Game/Gadgets/Assets/VinderTech_GliderChute/Glider_Stamina_Male/FX/Material/M_StaminaMale_Tail
graph {
 properties {
  allowTranslucentCustomDepthWrites=true
  bEnableResponsiveAA=true
  bUsedWithNiagaraMeshParticles=true
  bUsedWithNiagaraRibbons=true
  blendMode=BLEND_Translucent
  opacityMaskClipValue=0.001
  pixelDepthOffsetMode=PDOM_Legacy
  shadingModel=MSM_Unlit
  usedShadingModels=Unlit
 }
 nodes [
  Constant()
  n1 = PixelNormalWS()
  n2 = NamedRerouteUsage()
  n3 = Normalize(VectorInput=Transform(None=Constant3Vector() {constant=<rgba 0 0 1 0>}) {transformSourceType=TRANSFORMSOURCE_Instance})
  n4 = DotProduct(A=@n3, B=@n2)
  n5 = Normalize(VectorInput=Subtract(A=@n2, B=Multiply(A=@n4, B=@n3)))
  n6 = DotProduct(A=@n5, B=@n1)
  n7 = ScalarParameter() {group=Color, parameterName=Lighting_A_NDS_Offset}
  n8 = Saturate(None=Multiply(A=Add(A=@n6, B=@n7), B=ScalarParameter() {defaultValue=1, group=Color, parameterName=Lighting_A_NDS_Mul}))
  n9 = ScalarParameter() {defaultValue=1, group=Color, parameterName=Lighting_A_NDS_Power}
  n10 = VectorParameter() {group=Color, parameterName=Lighting_A_NDS_Remap, defaultValue=<rgba 0 1 0 1>}
  n11 = MaterialFunctionCall(Input=Saturate(None=Power(Base=@n8, Exp=@n9)), Input Low=Constant(), Input High=Constant() {r=1}, Target Low=@n10.R, Target High=@n10.R) {materialFunction=&RemapValueRange}
  n12 = Multiply(
   A=Add(A=DotProduct(A=@n5, B=@n1), B=ScalarParameter() {group=Color, parameterName=Lighting_C_NDS_Offset})
   B=ScalarParameter() {defaultValue=1, group=Color, parameterName=Lighting_C_NDS_Mul}
  )
  n13 = Saturate(None=Power(Base=Saturate(None=@n12), Exp=ScalarParameter() {defaultValue=1, group=Color, parameterName=Lighting_C_NDS_Power}))
  n14 = Constant()
  n15 = Constant() {r=1}
  n16 = VectorParameter() {group=Color, parameterName=Lighting_C_NDS_Remap, defaultValue=<rgba 0 1 0 1>}
  n17 = LinearInterpolate(
   A=VectorParameter() {group=Color, parameterName=Color0, defaultValue=<rgba 0 0 0 1>}.RGB
   B=LinearInterpolate(
    A=VectorParameter() {group=Color, parameterName=Color_Tip, defaultValue=<rgba 0.9 0.615 0.045 1>}.RGB
    B=VectorParameter() {group=Color, parameterName=Color_Tail, defaultValue=<rgba 1 0.6824 0.0677 1>}.RGB
    Alpha=Saturate(None=@n11.Result)
   )
   Alpha=Saturate(None=MaterialFunctionCall(Input=@n13, Input Low=@n14, Input High=@n15, Target Low=@n16.R, Target High=@n16.R) {materialFunction=&RemapValueRange}.Result)
  )
  n18 = Desaturation(
   None=MaterialFunctionCall(Hue Shift Percentage=ScalarParameter() {parameterName=Lighting_B_HueShift}, Texture=@n17) {materialFunction=&HueShift}.Result
   Fraction=ScalarParameter() {parameterName=Lighting_B_Desaturation}
  )
  n19 = Add(A=DotProduct(A=@n5, B=@n1), B=ScalarParameter() {group=Color, parameterName=Lighting_B_NDS_Offset})
  n20 = ScalarParameter() {defaultValue=1, group=Color, parameterName=Lighting_B_NDS_Mul}
  n21 = Power(Base=Saturate(None=Multiply(A=@n19, B=@n20)), Exp=ScalarParameter() {defaultValue=1, group=Color, parameterName=Lighting_B_NDS_Power})
  n22 = VectorParameter() {group=Color, parameterName=Lighting_B_NDS_Remap, defaultValue=<rgba 0 1 0 1>}
  n23 = Saturate(None=MaterialFunctionCall(Input=Saturate(None=@n21), Input Low=Constant(), Input High=Constant() {r=1}, Target Low=@n22.R, Target High=@n22.R) {materialFunction=&RemapValueRange}.Result)
  n24 = Multiply(A=Multiply(A=LinearInterpolate(A=@n18, B=@n17, Alpha=@n23), B=@n23)) {constB=0.5}
  n25 = Constant()
  n26 = Subtract(
   A=ComponentMask(None=SceneTexture() {sceneTextureId=PPI_CustomDepth}.Color) {r=true}
   B=ScalarParameter() {group=OutLine, parameterName=OutLineSampleBias}
  )
  n27 = ComponentMask(None=ScreenPosition().ViewportUV) {g=true, r=true}
  n28 = Subtract(A=Divide(A=ViewSize(), B=Constant() {r=320}), B=Constant() {r=2}) {constB=0}
  n29 = ViewSize()
  n30 = Divide(
   A=Divide(A=ComponentMask(None=@n29) {r=true}, B=ComponentMask(None=@n29) {g=true})
   B=MaterialFunctionCall() {materialFunction=/Engine/Functions/Engine_MaterialFunctions02/Utility/FOV}."FOV "
  )
  n31 = Multiply(A=Distance(A=ParticlePositionWS(), B=CameraPositionWS())) {constB=0.05}
  n32 = Multiply(
   A=Multiply(
    A=DynamicParameter() {parameterIndex=3, paramNames [RestAxis_X, RestAxis_Y, RestAxis_Z, LineThickness]}.LineThickness
    B=ScalarParameter() {defaultValue=10, group=OutLine, parameterName=LineThickness}
   )
   B=Power(Base=Divide(A=@n30, B=@n31)) {constExponent=0.6}
  )
  n33 = Multiply(A=Clamp(None=Ceil(None=@n28)) {maxDefault=2, minDefault=1}, B=Multiply(A=SceneTexelSize(), B=Clamp(None=@n32) {clampMode=CMODE_ClampMin}))
  n34 = Constant()
  n35 = AppendVector(A=ComponentMask(None=@n33) {desc=x, r=true}, B=@n34)
  n36 = ComponentMask(None=SceneTexture(UVs=Add(A=@n27, B=@n35) {desc=x+}) {sceneTextureId=PPI_CustomDepth}.Color) {r=true}
  n37 = AppendVector(A=@n34, B=ComponentMask(None=@n33) {desc=y, g=true}) {desc=y+}
  n38 = ComponentMask(None=SceneTexture(UVs=Add(A=@n27, B=@n37) {desc=y+}) {sceneTextureId=PPI_CustomDepth}.Color) {r=true}
  n39 = Constant() {r=-1}
  n40 = SceneTexture(UVs=Add(A=@n27, B=Multiply(A=@n35, B=@n39)) {desc=x-}) {sceneTextureId=PPI_CustomDepth}
  n41 = Add(A=Add(A=Subtract(A=@n26, B=@n36), B=Subtract(A=@n26, B=@n38)), B=Subtract(A=@n26, B=ComponentMask(None=@n40.Color) {r=true}))
  n42 = Add(A=@n27, B=Multiply(A=@n39, B=@n37)) {desc=y-}
  n43 = Subtract(A=@n26, B=ComponentMask(None=SceneTexture(UVs=@n42) {sceneTextureId=PPI_CustomDepth}.Color) {r=true})
  n44 = Saturate(None=Divide(A=Add(A=@n41, B=@n43), B=ScalarParameter() {defaultValue=-700, group=OutLine, parameterName=OutLineDivisor}) {constB=1})
  n45 = ScalarParameter() {defaultValue=1, group=OutLine, parameterName=OutLineScaleLine}
  n46 = Saturate(None=Power(Base=Multiply(A=@n44, B=@n45), Exp=ScalarParameter() {defaultValue=1, group=OutLine, parameterName=OutLinePowerline}))
  n47 = ShadingPathSwitch(Default=@n46, Deferred=@n46, Forward=@n46, Mobile=@n46)
  n48 = MaterialFunctionCall(Float2=TextureCoordinate()) {materialFunction=&BreakOutFloat2Components}
  n49 = DynamicParameter() {defaultValue=<rgba 1 0.16 1 1>, paramNames [BendAreaPower, Rigidity, DesatUVOffset, WhiteBoost]}
  n50 = Add(A=@n49.DesatUVOffset, B=ScalarParameter() {group=Color, parameterName=Desat_UVOffset})
  n51 = Saturate(None=Multiply(A=Add(A=@n48.G, B=@n50), B=ScalarParameter() {defaultValue=1.25, group=Color, parameterName=Desat_UVMul}))
  n52 = ScalarParameter() {defaultValue=1, group=Color, parameterName=Desat_UVPower}
  n53 = Multiply(A=OneMinus(None=Power(Base=@n51, Exp=@n52)), B=ScalarParameter() {group=Color, parameterName=Desat_Strength})
  n54 = Multiply(A=Desaturation(None=LinearInterpolate(A=@n24, B=@n25, Alpha=@n47), Fraction=@n53), B=LinearInterpolate(A=Add(A=@n53), Alpha=@n49.WhiteBoost))
  n55 = DynamicParameter() {parameterIndex=1, defaultValue=<rgba 1 0 1 1>, paramNames [Param1,Desat,Mul,Param4]}
  MaterialFunctionCall(Emissive=Multiply(A=Desaturation(None=@n54, Fraction=@n55.Desat), B=@n55.Mul)) {materialFunction=&MF_FX_EmissiveAdjustments}
  n56 = VectorParameter() {parameterName=FELightDir}
  NamedRerouteDeclaration(None=LinearInterpolate(A=SkyAtmosphereLightDirection(), B=@n56.RGB, Alpha=@n56.RGB)) {name=LightDir, nodeColor=<rgba 0 0.2031 1 1>}
  QualitySwitch(Default=@n47, Low=Constant(), High=@n47, Medium=@n47, Epic=@n47)
  n57 = DynamicParameter() {defaultValue=<rgba 1 0.16 1 1>, paramNames [BendAreaPower, Rigidity, DesatUVOffset, WhiteBoost]}
  n58 = DynamicParameter() {parameterIndex=2, paramNames [FollowPos_X, FollowPos_Y, FollowPos_Z, StretchDistance]}
  n59 = MaterialFunctionCall(Float2=TextureCoordinate()) {materialFunction=&BreakOutFloat2Components}
  n60 = DynamicParameter() {parameterIndex=3, paramNames [RestAxis_X, RestAxis_Y, RestAxis_Z, LineThickness]}
  n61 = MaterialFunctionCall!(
   Mask=Power(Base=OneMinus(None=@n59.G), Exp=@n57.BendAreaPower)
   FollowPos=AppendVector(A=AppendVector(A=@n58.FollowPos_X, B=@n58.FollowPos_X), B=@n58.FollowPos_Z)
   RestAxis=AppendVector(A=AppendVector(A=@n60.RestAxis_X, B=@n60.RestAxis_X), B=@n60.RestAxis_Z)
   MeshRootPos
   NormalWS
   StretchDistance=@n58.StretchDistance
   Rigidity=@n57.Rigidity
  ) {materialFunction=/Game/Effects/Niagara/Modules/FollowingDeform/Material/MF_NiagaraFollowingDeform}
  Reroute(None=Transform(None=@n61.NormalWS) {transformSourceType=TRANSFORMSOURCE_World, transformType=TRANSFORM_Tangent}) {desc=Normal}
  n62 = ParticleColor()
  n63 = MaterialFunctionCall(Float2=TextureCoordinate()) {materialFunction=&BreakOutFloat2Components}
  n64 = OneMinus(None=@n63.G)
  n65 = Clamp(None=Multiply(A=@n55.Param1, B=Add(A=@n62.B, B=@n64)), Min=@n55.Param1) {clampMode=CMODE_ClampMin}
  StaticSwitchParameter(True=Multiply(A=VertexNormalWS(), B=@n65), False=@n61.WPO) {parameterName=DisableWPO}
  n66 = Constant() {r=1}
  n67 = Constant()
  n68 = Constant()
  n69 = If(A=@n55.Param4, B=@n63.G, "A > B"=Constant() {r=1}, "A == B"=@n68, "A < B"=@n68)
  n70 = ScalarParameter() {defaultValue=1.25, group=Opacity, parameterName=ScaleOpacity}
  n71 = Power(Base=Saturate(None=Multiply(A=@n48.G, B=@n70)), Exp=ScalarParameter() {defaultValue=1, group=Opacity, parameterName=OpacityUVPower})
  n72 = Power(
   Base=DepthFade(FadeDistance=ScalarParameter() {defaultValue=50, group=Opacity, parameterName=DepthFadeDistance})
   Exp=ScalarParameter() {defaultValue=2, group=Opacity, parameterName=DepthFadePower}
  )
  n73 = Subtract(A=CameraPositionWS(), B=ParticlePositionWS())
  n74 = Normalize(VectorInput=Transform(None=Constant3Vector() {constant=<rgba 0 0 1 0>}) {transformSourceType=TRANSFORMSOURCE_Instance})
  n75 = Multiply(A=DotProduct(A=@n74, B=@n73), B=@n74)
  n76 = DotProduct(A=PixelNormalWS(), B=Normalize(VectorInput=Subtract(A=@n73, B=@n75)))
  n77 = MaterialFunctionCall(
   Input=Power(Base=Abs(None=@n76), Exp=ScalarParameter() {defaultValue=1, group=Opacity, parameterName=Opacity_AlignAspect_Power}) {constExponent=5}
   Target High=ScalarParameter() {defaultValue=0.85, group=Opacity, parameterName=Opacity_AlignAspect_MaxOpacity}
  ) {materialFunction=&RemapValueRange}
  n78 = Multiply(A=@n62.A, B=Multiply(A=Multiply(A=@n71, B=@n72), B=OneMinus(None=@n77.Result)))
  StaticSwitchParameter(
   True=Multiply(A=If(A=@n65, B=@n55.Param1, "A > B"=@n66, "A == B"=@n67, "A < B"=@n67) {equalsThreshold=0.1}, B=Multiply(A=@n69, B=@n78))
   False=@n78
  ) {parameterName=DisableWPO}
 ]
}
